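# Sound cue for equipping the lever-action shotgun with a swing, built from layered foley.
# A mixer plays three randomised layers at once: a boosted pump sound, a swing equip sound and a whoosh tail, each passing through a modulator.
# The pump layer's modulator varies pitch upward only.
# The cue uses weapon reload attenuation and the weapon foley sound class, with concurrency overridden to prevent new instances once the limit is reached.
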SoundCue /WeaponsUnvaulted/Assets/Audio/LeverAction_Shotgun/SC_Shotgun_Swing_Equip
root=Mixer_0
nodes [
 {id=Mixer_0, children [Modulator_0,Modulator_1,Modulator_2], properties {inputVolume [1.5,1,1]}}
 {id=Modulator_0, children [Random_2], properties {pitchMax=1.5, pitchMin=1}}
 {id=Modulator_1, children [Random_1]}
 {id=Modulator_2, children [Random_3]}
 {id=Random_1, children [WavePlayer_2,WavePlayer_3], properties {weights [1,1]}}
 {id=Random_2, children [WavePlayer_0,WavePlayer_1,WavePlayer_4], properties {weights [1,1,1]}}
 {id=Random_3, children [WavePlayer_12, WavePlayer_13], properties {weights [1,1]}}
 {
  id=WavePlayer_0
  properties {soundWaveAssetPtr=/WeaponsUnvaulted/Assets/Audio/LeverAction_Shotgun/SoundWaves/WPN_Shotgun_Swing_Reload_Pump_01}
 }
 {
  id=WavePlayer_1
  properties {soundWaveAssetPtr=/WeaponsUnvaulted/Assets/Audio/LeverAction_Shotgun/SoundWaves/WPN_Shotgun_Swing_Reload_Pump_02}
 }
 {
  id=WavePlayer_12
  properties {soundWaveAssetPtr=/WeaponsUnvaulted/Assets/Audio/LeverAction_Shotgun/SoundWaves/WPN_Shotgun_Swing_Reload_Whoosh_End_01}
 }
 {
  id=WavePlayer_13
  properties {soundWaveAssetPtr=/WeaponsUnvaulted/Assets/Audio/LeverAction_Shotgun/SoundWaves/WPN_Shotgun_Swing_Reload_Whoosh_End_02}
 }
 {id=WavePlayer_2, properties {soundWaveAssetPtr=/WeaponsUnvaulted/Assets/Audio/LeverAction_Shotgun/SoundWaves/Shotgun_Swing_Equip_01}}
 {id=WavePlayer_3, properties {soundWaveAssetPtr=/WeaponsUnvaulted/Assets/Audio/LeverAction_Shotgun/SoundWaves/Shotgun_Swing_Equip_02}}
 {
  id=WavePlayer_4
  properties {soundWaveAssetPtr=/WeaponsUnvaulted/Assets/Audio/LeverAction_Shotgun/SoundWaves/WPN_Shotgun_Swing_Reload_Pump_03}
 }
]
properties {
 attenuationSettings=&Fort_Weapon_Reload_Att
 bOverrideConcurrency=true
 maxDistance=2250
 pitchMultiplier=1.1
 soundClassObject=&Fort_WeaponFoley
 volumeMultiplier=0.5
 attenuationOverrides {falloffDistance=4100, attenuationShapeExtents {x=900}}
 concurrencyOverrides {resolutionRule=PreventNew}
}
