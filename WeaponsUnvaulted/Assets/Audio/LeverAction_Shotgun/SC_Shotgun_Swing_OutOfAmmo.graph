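# Sound cue for the lever-action shotgun's out-of-ammo sound, routed to a third-person ranged weapon fire sound class.
# A random node picks one of four equally weighted out-of-ammo waves, then a modulator adds slight volume variation with fixed pitch.
# Concurrency overrides allow one instance per owner, blocking new plays while one is active and throttling retriggers.
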SoundCue /WeaponsUnvaulted/Assets/Audio/LeverAction_Shotgun/SC_Shotgun_Swing_OutOfAmmo
root=Modulator_0
nodes [
 {id=Modulator_0, children [Random_1], properties {pitchMax=1, pitchMin=1, volumeMax=1.1, volumeMin=1}}
 {id=Random_1, children [WavePlayer_0,WavePlayer_5,WavePlayer_6,WavePlayer_7], properties {weights [1,1,1,1]}}
 {id=WavePlayer_0, properties {soundWaveAssetPtr=/WeaponsUnvaulted/Assets/Audio/LeverAction_Shotgun/SoundWaves/Shotgun_OutofAmmo_04}}
 {id=WavePlayer_5, properties {soundWaveAssetPtr=/WeaponsUnvaulted/Assets/Audio/LeverAction_Shotgun/SoundWaves/Shotgun_OutofAmmo_03}}
 {id=WavePlayer_6, properties {soundWaveAssetPtr=/WeaponsUnvaulted/Assets/Audio/LeverAction_Shotgun/SoundWaves/Shotgun_OutofAmmo_02}}
 {id=WavePlayer_7, properties {soundWaveAssetPtr=/WeaponsUnvaulted/Assets/Audio/LeverAction_Shotgun/SoundWaves/Shotgun_OutofAmmo_01}}
]
properties {
 bOverrideConcurrency=true
 maxDistance=2097152
 soundClassObject=&Fort_WeaponRanged_Fire_3P_PSM
 volumeMultiplier=0.65
 attenuationOverrides {falloffDistance=1600}
 concurrencyOverrides {bLimitToOwner=true, maxCount=1, resolutionRule=PreventNew, retriggerTime=0.3}
}
