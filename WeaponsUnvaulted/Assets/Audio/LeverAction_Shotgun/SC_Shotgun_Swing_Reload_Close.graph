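# SoundCue for the closing foley sound of the lever-action shotgun's swing reload.
# A single wave player feeds the Shotgun_Swing_Reload_Close wave through a Modulator left at default settings.
# It plays in the weapon foley sound class with reload attenuation, a reduced volume and slightly raised pitch.
# A concurrency override prevents new instances from starting once the limit is reached.
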SoundCue /WeaponsUnvaulted/Assets/Audio/LeverAction_Shotgun/SC_Shotgun_Swing_Reload_Close
root=Modulator_1
nodes [
 {id=Modulator_1, children [WavePlayer_1]}
 {id=WavePlayer_1, properties {soundWaveAssetPtr=/WeaponsUnvaulted/Assets/Audio/LeverAction_Shotgun/SoundWaves/Shotgun_Swing_Reload_Close}}
]
properties {
 attenuationSettings=&Fort_Weapon_Reload_Att
 bOverrideConcurrency=true
 maxDistance=2250
 pitchMultiplier=1.1
 soundClassObject=&Fort_WeaponFoley
 volumeMultiplier=0.45
 attenuationOverrides {falloffDistance=4100, attenuationShapeExtents {x=900}}
 concurrencyOverrides {resolutionRule=PreventNew}
}
