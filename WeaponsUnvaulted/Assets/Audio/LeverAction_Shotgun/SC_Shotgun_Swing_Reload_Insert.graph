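# Sound cue for inserting a shell while reloading the lever-action swing shotgun.
# A Random node picks one of three equally weighted shell-insert wave variations, then a Modulator applies variation to the result.
# It plays in the weapon foley sound class with reload attenuation and reduced volume.
# A concurrency override is set to prevent new instances once the limit is reached.
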SoundCue /WeaponsUnvaulted/Assets/Audio/LeverAction_Shotgun/SC_Shotgun_Swing_Reload_Insert
root=Modulator_0
nodes [
 {id=Modulator_0, children [Random_1]}
 {id=Random_1, children [WavePlayer_0,WavePlayer_1,WavePlayer_5], properties {weights [1,1,1]}}
 {
  id=WavePlayer_0
  properties {soundWaveAssetPtr=/WeaponsUnvaulted/Assets/Audio/LeverAction_Shotgun/SoundWaves/WPN_Shotgun_Swing_Reload_Insert_Shell_01}
 }
 {
  id=WavePlayer_1
  properties {soundWaveAssetPtr=/WeaponsUnvaulted/Assets/Audio/LeverAction_Shotgun/SoundWaves/WPN_Shotgun_Swing_Reload_Insert_Shell_02}
 }
 {
  id=WavePlayer_5
  properties {soundWaveAssetPtr=/WeaponsUnvaulted/Assets/Audio/LeverAction_Shotgun/SoundWaves/WPN_Shotgun_Swing_Reload_Insert_Shell_03}
 }
]
properties {
 attenuationSettings=&Fort_Weapon_Reload_Att
 bOverrideConcurrency=true
 maxDistance=2250
 soundClassObject=&Fort_WeaponFoley
 volumeMultiplier=0.601
 attenuationOverrides {falloffDistance=4100, attenuationShapeExtents {x=900}}
 concurrencyOverrides {resolutionRule=PreventNew}
}
